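# Translucent, lit Niagara material for derez-style cube surfing VFX on sprites, ribbons and mesh particles.
# A time-animated hash from particle color length and a random offset makes a flickering brightness, multiplied by a cube gradient texture and a parameter or particle color.
# An HDR-remapped Fresnel glow is added, then divided by eye adaptation.
# Particle alpha and the inverted gradient feed a SmoothStep mask for a gradient-driven dissolve.
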Material /Game/Effects/Niagara/MasterMaterials/Derez/M_CubeSurfing_VFX
graph {
 properties {
  bAutomaticallySetUsageInEditor=false
  bUsedWithNiagaraMeshParticles=true
  bUsedWithNiagaraRibbons=true
  bUsedWithNiagaraSprites=true
  blendMode=BLEND_Translucent
  maxWorldPositionOffsetDisplacement=800
  translucencyLightingMode=TLM_Surface
  usedShadingModels=Default Lit
 }
 nodes [
  Constant() {r=0.5}
  Constant() {r=0.2}
  n1 = TextureSampleParameter2D() {group=Color_Shading, parameterName=CubeGradient, texture=&/T_CubeGradient}
  n2 = LinearInterpolate(Alpha=@n1.R) {constA=0.5}
  n3 = DynamicParameter() {defaultValue=<rgba 0 0 1 0>, paramNames [FresnelGlow, UseParticleColor, OffsetRandom, ParticleHashMin_Cd]}
  n4 = LinearInterpolate(
   A=ScalarParameter() {defaultValue=0.3, parameterName=HashMin_Cd}
   B=DynamicParameter() {parameterIndex=1, defaultValue=<rgba 0.3 0 0 0>, paramNames [HashMin_Cd,2,3,4]}.HashMin_Cd
   Alpha=@n3.ParticleHashMin_Cd
  )
  n5 = ParticleColor()
  n6 = Length(None=@n5.RGB) {is=class MaterialExpressionLength}
  n7 = Multiply(A=@n6) {constB=5}
  n8 = Multiply(A=Add(A=Time(), B=@n3.OffsetRandom), B=ScalarParameter() {defaultValue=1, parameterName=DerezSpeed})
  n9 = Power(Base=Frac(None=Add(A=@n7, B=@n8)))
  n10 = LinearInterpolate(
   A=Multiply(A=@n3.OffsetRandom, B=VectorParameter() {parameterName=Color, defaultValue=<rgba 0.0494 0.6103 1 1>}.RGB)
   B=@n5.RGB
   Alpha=@n3.UseParticleColor
  )
  n11 = Multiply(
   A=NamedRerouteDeclaration(None=LinearInterpolate(A=@n4, Alpha=@n9) {constA=0.2}) {name=SimpleHashNoise_Cd, nodeColor=<rgba 0 1 0.9531 1>}
   B=@n10
  )
  n12 = Constant() {r=2}
  n13 = NamedRerouteDeclaration(None=MaterialFunctionCall(
    LDR Input=MaterialFunctionCall(Brightness=@n3.FresnelGlow, Exponent=@n12) {materialFunction=&MF_Fresnel_FX}.Fresnel
    HDR Dynamic Range=Constant() {r=1}
    HDR Intensity Multiplier=Constant() {r=100}
   ) {materialFunction=&DeriveHDRfromLDR}.RemappedRange) {name=Fresnel, nodeColor=<rgba 1 0 0.6797 1>}
  Divide(A=Add(A=Multiply(A=@n2, B=@n11), B=Multiply(A=@n10, B=@n13)), B=EyeAdaptation().EyeAdaptation)
  Frac(None=@n6)
  MaterialFunctionCall() {materialFunction=&DebugTimeSine}
  n14 = ParticleColor()
  SmoothStep(Min=Subtract(A=OneMinus(None=@n1.R), B=@n14.A), Value=@n14.A)
 ]
}
