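# Unlit masked Niagara material for the derez spatial travel trail effect, usable on ribbons, sprites and mesh particles.
# Grid-square and radar-grid textures, hashed cells animated by time, and stepped UV gradient and circle masks build the dissolve and two-color emissive.
# A Swizzle parameter blends UVs toward swapped axes to reorient the pattern.
# Particle alpha scales a mask, and a dynamic parameter scales SplineThicken width.
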Material /Game/Effects/Niagara/MasterMaterials/Derez/M_Derez_SpatialTravel_Trail
graph {
 properties {
  bAutomaticallySetUsageInEditor=false
  bUsedWithNiagaraMeshParticles=true
  bUsedWithNiagaraRibbons=true
  bUsedWithNiagaraSprites=true
  blendMode=BLEND_Masked
  opacityMaskClipValue=0.1
  shadingModel=MSM_Unlit
  usedShadingModels=Unlit
 }
 nodes [
  n1 = Time()
  n2 = TextureCoordinate() {uTiling=0.5, vTiling=0.5}
  n3 = TextureSample(UVs=LinearInterpolate(
    A=@n2
    B=MaterialFunctionCall(XY=@n2) {materialFunction=&Swizzle}.YX
    Alpha=ScalarParameter() {parameterName=Swizzle}
   )) {
   samplerSource=SSM_Wrap_WorldGroupSettings
   samplerType=SAMPLERTYPE_LinearColor
   texture=/Game/Packages/Fortress_SharedMaterials/Textures/T_Slowfield_GridSquares_01
  }
  ConstantBiasScale(None=Sine(None=Add(A=@n1, B=@n3.R)))
  n4 = TextureCoordinate()
  n5 = ScalarParameter() {parameterName=Swizzle}
  n6 = Constant() {r=8}
  n7 = Multiply(
   A=LinearInterpolate(A=@n4, B=MaterialFunctionCall(XY=@n4) {materialFunction=&Swizzle}.YX, Alpha=@n5)
   B=@n6
  )
  n8 = TextureCoordinate()
  n9 = MaterialFunctionCall(XY=@n8) {materialFunction=&Swizzle}
  n10 = Floor(None=Multiply(A=LinearInterpolate(A=@n8, B=@n9.YX, Alpha=@n5), B=@n6))
  n11 = Length(None=Subtract(A=Divide(A=@n10, B=@n6)) {constB=0.45}) {is=class MaterialExpressionLength}
  n12 = Step(Y=@n11) {constX=0.5}
  n13 = Multiply(A=MaterialFunctionCall(Vec2=Ceil(None=@n7)) {materialFunction=&MF_Hash}.Vec3, B=Ceil(None=@n12))
  n14 = Sine(None=Add(A=Time(), B=@n13))
  n15 = OneMinus(None=@n11)
  n16 = Power(Base=Multiply(A=@n15, B=@n12)) {constExponent=5}
  LinearInterpolate(A=Multiply(A=@n13, B=ConstantBiasScale(None=@n14)), B=@n16, Alpha=@n15)
  MaterialFunctionCall(Emissive=NamedRerouteUsage()) {materialFunction=&MF_FX_EmissiveAdjustments}
  n17 = DynamicParameter() {paramNames [WidthWPO_Scale, Param2, Param3, Param4]}
  MaterialFunctionCall(WidthBase=Multiply(A=Constant() {r=10}, B=@n17.WidthWPO_Scale), WidthTip=Multiply(A=Constant() {r=80}, B=@n17.WidthWPO_Scale)) {materialFunction=&SplineThicken}
  Multiply(A=Multiply(A=NamedRerouteUsage(), B=ParticleColor().A), B=NamedRerouteUsage())
  n18 = TextureCoordinate() {uTiling=0.5, vTiling=0.5}
  n19 = Ceil(None=Abs(None=@n18))
  n20 = Add(A=Time() {period=1}, B=Add(A=ComponentMask(None=@n19) {r=true}, B=ComponentMask(None=@n19) {g=true}))
  n21 = Multiply(A=ScalarParameter() {defaultValue=0.5, parameterName=Sine Speed}, B=@n20)
  n22 = ScalarParameter() {parameterName=Swizzle}
  n23 = TextureSample(UVs=LinearInterpolate(A=@n18, B=MaterialFunctionCall(XY=@n18) {materialFunction=&Swizzle}.YX, Alpha=@n22)) {samplerType=SAMPLERTYPE_LinearColor, texture=&T_Slowfield_GridSquares_01}
  n24 = Sine(None=Frac(None=Add(A=@n21, B=@n23.R)))
  n25 = Multiply(A=TextureCoordinate(), B=ScalarParameter() {defaultValue=8, parameterName=LinesScale})
  n26 = TextureSample(UVs=LinearInterpolate(A=@n25, B=MaterialFunctionCall(XY=@n25) {materialFunction=&Swizzle}.YX, Alpha=@n22)) {samplerType=SAMPLERTYPE_LinearColor, texture=/Game/Environments/Hexmap/HexSelection/T_Hex_RadarGrid02}
  n27 = OneMinus(None=Saturate(None=LinearInterpolate(Alpha=@n26.R) {constA=-2, constB=10}))
  n28 = TextureCoordinate()
  n29 = ComponentMask(None=LinearInterpolate(
    A=@n28
    B=MaterialFunctionCall(XY=@n28) {materialFunction=&Swizzle}.YX
    Alpha=ScalarParameter() {parameterName=Swizzle}
   )) {g=true}
  n30 = Constant() {r=8}
  n31 = Multiply(A=@n29, B=@n30)
  n32 = Saturate(None=Divide(A=Floor(None=@n31), B=@n30))
  n33 = Multiply(A=@n27, B=@n32)
  Multiply(A=@n24, B=@n33)
  n34 = Multiply(B=@n3.R) {constA=1}
  NamedRerouteDeclaration(None=LinearInterpolate(A=@n34, Alpha=NamedRerouteUsage())) {name=Derez_Mask, nodeColor=<rgba 0.0076 0.75 1 1>}
  n35 = Saturate(None=LinearInterpolate(Alpha=@n26.R) {constA=-2, constB=10})
  n36 = Multiply(A=@n24, B=@n35)
  n37 = TextureCoordinate()
  n38 = MaterialFunctionCall(XY=@n37) {materialFunction=&Swizzle}
  n39 = ScalarParameter() {parameterName=Swizzle}
  n40 = Ceil(None=Multiply(A=LinearInterpolate(A=@n37, B=@n38.YX, Alpha=@n39), B=Constant() {r=8}))
  n41 = Frac(None=Add(A=MaterialFunctionCall(Vec2=@n40) {materialFunction=&MF_Hash}.Vec2, B=@n20))
  n42 = LinearInterpolate(A=Multiply(A=Sine(None=@n41), B=@n32), Alpha=@n32) {constB=0.5}
  n43 = Add(A=Saturate(None=LinearInterpolate(A=@n36, B=@n35, Alpha=@n32)), B=Multiply(A=LinearInterpolate(B=@n33, Alpha=@n42)) {constB=2.5})
  NamedRerouteDeclaration(None=Power(Base=Multiply(A=@n43) {constB=5})) {name=Derez_Dissolve, nodeColor=<rgba 0.4766 1 0 1>}
  n44 = Multiply(A=@n34) {constB=1.2}
  n45 = NamedRerouteUsage()
  n46 = LinearInterpolate(
   A=VectorParameter() {group=Derez, parameterName=Color_02, defaultValue=<rgba 0 1.9099 3 1>}.RGB
   B=VectorParameter() {group=Derez, parameterName=Color_01, defaultValue=<rgba 0.0166 0.0885 0.1138 1>}.RGB
   Alpha=Multiply(A=Saturate(None=@n44), B=Power(Base=@n45))
  )
  n47 = Multiply(A=NamedRerouteUsage()) {constB=2.5}
  NamedRerouteDeclaration(None=LinearInterpolate(A=Multiply(A=@n46) {constB=5}, B=@n46, Alpha=Saturate(None=@n47))) {name=Emissive, nodeColor=<rgba 0 0.3437 1 1>}
  NamedRerouteDeclaration(None=@n27) {name=Grid, nodeColor=<rgba 1 0 0.5859 1>}
  n48 = TextureCoordinate()
  n49 = LinearInterpolate(A=@n48, B=MaterialFunctionCall(XY=@n48) {materialFunction=&Swizzle}.YX, Alpha=@n5)
  n50 = Multiply(A=Frac(None=ComponentMask(None=@n49) {r=true})) {constB=2}
  n51 = Multiply(A=Abs(None=Subtract(A=@n50))) {constB=4}
  n52 = OneMinus(None=Divide(A=Floor(None=@n51)) {constB=4})
  NamedRerouteDeclaration(None=Multiply(A=@n16, B=Power(Base=@n52) {constExponent=3})) {name=Circle_Mask, nodeColor=<rgba 0.2656 0 1 1>}
  NamedRerouteDeclaration(None=@n32) {name=U_Gradient, nodeColor=<rgba 0.1484 1 0 1>}
  OneMinus(None=@n29)
 ]
}
