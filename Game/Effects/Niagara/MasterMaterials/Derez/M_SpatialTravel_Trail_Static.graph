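# Unlit, two-sided translucent Niagara material for a spatial travel trail with a blocky "derez" look, usable on sprites, ribbons and mesh particles.
# Pixelated scrolling noise blends two color parameters into an emissive tinted by particle color.
# Named reroutes carry masks built from a stepped U gradient, hashed time-animated cells, hex-grid lines and grid-square textures.
# These form the derez dissolve, with particle alpha available for fading.
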Material /Game/Effects/Niagara/MasterMaterials/Derez/M_SpatialTravel_Trail_Static
graph {
 properties {
  bAutomaticallySetUsageInEditor=false
  bUsedWithNiagaraMeshParticles=true
  bUsedWithNiagaraRibbons=true
  bUsedWithNiagaraSprites=true
  blendMode=BLEND_Translucent
  opacityMaskClipValue=0.1
  shadingModel=MSM_Unlit
  twoSided=true
  usedShadingModels=Unlit
 }
 nodes [
  n1 = Time()
  n2 = TextureCoordinate() {uTiling=0.5, vTiling=0.5}
  n3 = Multiply(A=Time(), B=Constant2Vector() {r=0.2})
  n4 = Constant() {r=8}
  n5 = Frac(None=Multiply(A=Add(A=@n2, B=@n3), B=@n4))
  n6 = TextureSample(UVs=Add(A=@n2, B=Divide(A=@n5, B=@n4))) {
   samplerSource=SSM_Wrap_WorldGroupSettings
   samplerType=SAMPLERTYPE_LinearColor
   texture=/Game/Packages/Fortress_SharedMaterials/Textures/T_Slowfield_GridSquares_01
  }
  ConstantBiasScale(None=Sine(None=Add(A=@n1, B=@n6.R)))
  MaterialFunctionCall(Emissive=NamedRerouteUsage()) {materialFunction=&MF_FX_EmissiveAdjustments}
  Multiply(A=NamedRerouteUsage(), B=ParticleColor().A)
  n7 = Multiply(B=@n6.R) {constA=1}
  NamedRerouteDeclaration(None=LinearInterpolate(A=@n7, Alpha=NamedRerouteUsage())) {name=Derez_Mask, nodeColor=<rgba 0.0076 0.75 1 1>}
  n8 = ScalarParameter() {defaultValue=8, parameterName=RezScale}
  n9 = Floor(None=Multiply(A=TextureCoordinate(), B=@n8))
  n10 = Time()
  n11 = Constant2Vector() {r=-0.2}
  n12 = TextureSample(UVs=Add(A=Divide(A=@n9, B=@n8), B=Multiply(A=@n10, B=@n11))) {
   samplerSource=SSM_Wrap_WorldGroupSettings
   samplerType=SAMPLERTYPE_LinearColor
   texture=/Game/Characters/Player/Male/Medium/Bodies/M_MED_EmeraldGlassGreen/Textures/T_CowlNoise
  }
  n13 = NamedRerouteUsage()
  n14 = LinearInterpolate(
   A=VectorParameter() {group=Derez, parameterName=Color_01, defaultValue=<rgba 0.0166 0.0885 0.1138 1>}.RGB
   B=VectorParameter() {group=Derez, parameterName=Color_02, defaultValue=<rgba 0 1.9099 3 1>}.RGB
   Alpha=Multiply(A=@n12.RGB, B=Power(Base=@n13))
  )
  n15 = Multiply(A=@n14) {constB=5}
  n16 = NamedRerouteUsage()
  NamedRerouteDeclaration(None=Multiply(A=LinearInterpolate(A=@n14, B=@n15, Alpha=@n16), B=ParticleColor().RGB)) {name=Emissive, nodeColor=<rgba 0 0.3437 1 1>}
  n17 = TextureCoordinate()
  n18 = ScalarParameter() {defaultValue=8, parameterName=LinesScale}
  n19 = LinearInterpolate(Alpha=TextureSample(UVs=Multiply(A=@n17, B=@n18)) {samplerType=SAMPLERTYPE_LinearColor, texture=/Game/Environments/Hexmap/HexSelection/T_Hex_RadarGrid02}.R) {constA=-2, constB=10}
  NamedRerouteDeclaration(None=OneMinus(None=Saturate(None=@n19))) {name=Grid, nodeColor=<rgba 1 0 0.5859 1>}
  n20 = TextureCoordinate()
  n21 = Constant() {r=8}
  n22 = Multiply(A=OneMinus(None=ComponentMask(None=@n20) {r=true}), B=@n21)
  n23 = Saturate(None=Divide(A=Floor(None=@n22), B=@n21))
  NamedRerouteDeclaration(None=@n23) {name=U_Gradient, nodeColor=<rgba 0.1484 1 0 1>}
  n24 = Multiply(A=TextureCoordinate(), B=Constant() {r=8})
  n25 = TextureCoordinate() {uTiling=0.5, vTiling=0.5}
  n26 = Ceil(None=Abs(None=@n25))
  n27 = Add(A=Time() {period=1}, B=Add(A=ComponentMask(None=@n26) {r=true}, B=ComponentMask(None=@n26) {g=true}))
  n28 = Add(A=MaterialFunctionCall(Vec2=Ceil(None=@n24)) {materialFunction=&MF_Hash}.Vec2, B=@n27)
  n29 = Multiply(A=Sine(None=Frac(None=@n28)), B=@n23)
  NamedRerouteDeclaration(None=ComponentMask(None=LinearInterpolate(A=@n29, Alpha=@n23) {constB=0.5}) {r=true}) {name=Gradient_Dissolve, nodeColor=<rgba 0 1 0.625 1>}
  n30 = NamedRerouteUsage()
  n31 = Multiply(A=@n12.RGB, B=@n30)
  n32 = Power(Base=@n30, Exp=ScalarParameter() {defaultValue=5, parameterName=Power Dissolve}) {constExponent=5}
  n33 = NamedRerouteUsage()
  n34 = NamedRerouteUsage()
  NamedRerouteDeclaration(None=Add(A=Add(A=@n31, B=@n32), B=LinearInterpolate(B=@n33, Alpha=@n34))) {name=Derez_Dissolve, nodeColor=<rgba 0.4766 1 0 1>}
  NamedRerouteDeclaration(None=ParticleColor().A) {name=AlphaDissolve, nodeColor=<rgba 1 0 0.5156 1>}
  NamedRerouteUsage()
  NamedRerouteUsage()
  NamedRerouteUsage()
  Saturate(None=Multiply(A=@n7) {constB=1.2})
  n35 = Multiply(A=ScalarParameter() {defaultValue=0.5, parameterName=Sine Speed}, B=@n27)
  n36 = TextureSample(UVs=@n25) {samplerType=SAMPLERTYPE_LinearColor, texture=&T_Slowfield_GridSquares_01}
  Sine(None=Frac(None=Add(A=@n35, B=@n36.R)))
  Step() {constX=0.5}
 ]
}
